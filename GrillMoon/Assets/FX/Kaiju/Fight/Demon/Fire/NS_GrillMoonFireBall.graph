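# Niagara effect for GrillMoon's demon kaiju fireball, sized by FX scale and switched from orange to blue by Is Blue.
# Mesh mask flame, eye trail and Minimal layers combine with a spinning flare, pulsing glow and light.
# Sine/cosine offsets twist two ribbon smoke emitters, one spawning by rate, the other by distance.
# The rate-based one feeds trailing embers through particle reads, alongside flipbook core smoke.
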NiagaraSystem /GrillMoon/Assets/FX/Kaiju/Fight/Demon/Fire/NS_GrillMoonFireBall
dependencies {
 usedDataInterfaces [
  &NiagaraDataInterfaceCurve
  &NiagaraDataInterfaceParticleRead
  &NiagaraDataInterfaceCamera
  &NiagaraDataInterfaceSkeletalMesh
 ]
 usedDynamicInputs [
  &SimulationPosition
  &VectorLength
  &MakeCustomFloatFromBool
 ]
 usedModules [
  &self:SystemSpawnScript.SystemScriptSource.SystemScriptGraph.NiagaraNodeAssignment_1.SetVariables_#1
  &self:SystemSpawnScript.SystemScriptSource.SystemScriptGraph.NiagaraNodeAssignment_2.SetVariables_#2
  &self:SystemSpawnScript.SystemScriptSource.SystemScriptGraph.NiagaraNodeAssignment_3.SetVariables_#3
  &self:SystemSpawnScript.SystemScriptSource.SystemScriptGraph.NiagaraNodeAssignment_4.SetVariables_#4
  &self:SystemSpawnScript.SystemScriptSource.SystemScriptGraph.NiagaraNodeAssignment_7.SetVariables_#5
  &self:SystemSpawnScript.SystemScriptSource.SystemScriptGraph.NiagaraNodeAssignment_8.SetVariables_#6
  &SystemState
  &self:SystemSpawnScript.SystemScriptSource.SystemScriptGraph.NiagaraNodeAssignment_6.SetVariables_#7
  &self:SystemSpawnScript.SystemScriptSource.SystemScriptGraph.NiagaraNodeAssignment_5.SetVariables_#8
  &self:SystemSpawnScript.SystemScriptSource.SystemScriptGraph.NiagaraNodeAssignment_0.SetVariables_#9
 ]
}
emitters {
 EyeTrail {
  emitterUpdateScript [
   &EmitterState(Inactive Response=::NewEnumerator1, Life Cycle Mode=::NewEnumerator1, Loop Behavior=::NewEnumerator1, Loop Duration=1, UseLoopDelay=false)
   &SpawnBurst_Instantaneous(Use Spawn Probability=false, UseLoopCountLimit=true)
  ]
  particleSpawnScript [
   &InitializeParticle(
    Color=&LerpLinearColors(StartColor=<rgba 21.3542 10.3212 0 1>, EndColor=<rgba 0 14.2361 21.3542 1>, LerpFactor=$System.Is Blue)
    Lifetime Max=2.5
    Lifetime Min=1.5
    Lifetime Mode=::NewEnumerator1
    Mesh Scale Mode=::NewEnumerator3
    Mesh Uniform Scale=0.7
    Position Offset=&Multiply_VectorByFloat(Vector=<-2.5943 0 1.6394>, Float=$System.FX scale)
    Sprite Size Mode=::NewEnumerator3
    Uniform Sprite Size=5
    UsePositionOffset=true
    Write Material Random=true
   )
   &InitialMeshOrientation(Mesh Axis To Orient=<0 1 0>, Mesh Orientation Mode=::NewEnumerator2, Use Rotation Vector=false)
   &CameraOffset(Camera Offset Amount=2)
   /SetVariables(Particles.VisibilityTag=&MakeCustomIntFromBool(Bool=&SetBoolByFloatComparison(A=$System.Is Blue, B=0.5), True INT=1, False INT=0)) {isSetParameters=true}
  ]
  particleUpdateScript [
   &ParticleState(Kill Particles When Lifetime Has Elapsed=false, Let Infinitely Lived Particles Die When Emitter Deactivates=true)
   &ScaleMeshSize(Scale Factor=&VectorFromFloat(Value=$System.FX scale))
   &DynamicMaterialParameters(
    Index 0 Param 3=1
    Module.Index0Param0Write=true
    Module.Index0Param1Write=true
    Module.Index0Param2Write=true
    Module.Index0Param3Write=true
    Module.Index1Param0Write=true
    Module.Index1Param1Write=true
    Module.Index1Param2Write=true
    Module.Index1Param3Write=true
    Module.Index2Param0Write=true
    Module.Index2Param1Write=true
    Module.Index2Param2Write=true
    Module.Index2Param3Write=true
    Write Parameter Index 1=true
    Write Parameter Index 2=true
   )
   &SkeletalMeshLocation(
    Bone Sampling Mode=::NewEnumerator4
    MaskSpawnGroup=false
    Orientation Sampling=::NewEnumerator1
    Sampled Position Offset=<0 -5 0>
    Skeletal Mesh=$System.Skeletal Mesh
   ) {enabled=false}
  ]
  rendererClasses [&NiagaraMeshRendererProperties]
  renderers [
   {rendererClass=&NiagaraMeshRendererProperties, rendererIndex=0}
   {rendererClass=&NiagaraMeshRendererProperties, rendererIndex=1}
  ]
 }
 Flare {
  emitterUpdateScript [
   &EmitterState(
    Inactive Response=::NewEnumerator1
    Life Cycle Mode=::NewEnumerator1
    Loop Behavior=::NewEnumerator1
    Loop Duration Mode=::NewEnumerator1
    UseLoopDelay=false
   )
   &SpawnBurst_Instantaneous(Spawn Count=1, Use Spawn Probability=false)
  ]
  particleSpawnScript [
   &InitializeParticle(
    Color=&LerpLinearColors(StartColor=<rgba 3 1 0 1>, EndColor=<rgba 0 2 3 1>, LerpFactor=$System.Is Blue)
    Sprite Size Mode=::NewEnumerator3
    Uniform Sprite Size=50
    UsePositionOffset=false
    Write Material Random=true
   ) {module=InitializeParticle001}
   /SetVariables() {isSetParameters=true}
   &CameraOffset(Camera Offset Amount=50) {enabled=false}
  ]
  particleUpdateScript [
   &ParticleState(Kill Particles When Lifetime Has Elapsed=false, Let Infinitely Lived Particles Die When Emitter Deactivates=true)
   &ScaleSpriteSize(Scale Sprite Size Mode=::NewEnumerator0, Uniform Scale Factor=$System.FX scale)
   &SpriteRotationRate(Rotation Rate=150)
  ]
  rendererClasses [&NiagaraSpriteRendererProperties]
  renderers [{rendererClass=&NiagaraSpriteRendererProperties}]
 }
 Illumination {
  emitterUpdateScript [
   &EmitterState()
   &SpawnBurst_Instantaneous(Use Spawn Probability=false, UseLoopCountLimit=true)
  ]
  particleSpawnScript [
   &InitializeParticle(
    Color=&LerpLinearColors(StartColor=<rgba 1 0.3634 0 1>, EndColor=<rgba 0 0.6667 1 1>, LerpFactor=$System.Is Blue)
    Lifetime=5
    Sprite Size Mode=::NewEnumerator3
    Uniform Sprite Size=70
    UsePositionOffset=false
    Write Material Random=true
   )
   &CameraOffset(Camera Offset Amount=&Multiply_Float(A=20, B=$System.FX scale))
  ]
  particleUpdateScript [
   &ParticleState(Kill Particles When Lifetime Has Elapsed=false, Let Infinitely Lived Particles Die When Emitter Deactivates=true)
   &ScaleSpriteSize(Scale Sprite Size Mode=::NewEnumerator0, Uniform Scale Factor=$System.FX scale)
   &ScaleColor(
    Scale RGB=&VectorFromFloat(Value=&Waveform(Global Amplitude Scale=0.1, "Waveform [1]"=::NewEnumerator0, "[1] Frequency"=1, "[1] Amplitude Min/Max"=<0.2 1>, "[1] Bias"=0.5))
    ScaleA=true
    ScaleRGB=true
   )
   &FN_Light_Attributes(
    Color=$Particles.Color
    Create Color Attribute=true
    Create Exponent Attribute=true
    Create Radius Attribute=true
    Radius=&Multiply_Float(A=80, B=$System.FX scale)
    Scale Color=250
   )
  ]
  rendererClasses [&NiagaraSpriteRendererProperties, &NiagaraLightRendererProperties]
  renderers [
   {rendererClass=&NiagaraSpriteRendererProperties}
   {rendererClass=&NiagaraLightRendererProperties, rendererIndex=1}
  ]
 }
 MaskFlame {
  emitterUpdateScript [
   &EmitterState(
    Inactive Response=::NewEnumerator1
    Life Cycle Mode=::NewEnumerator1
    Loop Behavior=::NewEnumerator1
    Loop Duration Mode=::NewEnumerator1
    UseLoopDelay=false
   )
   &SpawnBurst_Instantaneous(Use Spawn Probability=false, UseLoopCountLimit=false)
  ]
  particleSpawnScript [
   &InitializeParticle(
    Color=<rgba 1 1 1 2>
    Mesh Scale=&Multiply_Vector(A=<0.7 2 0.7>, B=&VectorFromFloat(Value=0.8))
    Mesh Scale Mode=::NewEnumerator4
    Position Offset=&Multiply_Vector(A=<3.99 0 2.1516>, B=&VectorFromFloat(Value=$System.FX scale))
    UsePositionOffset=true
    Write Material Random=true
   )
   &InitialMeshOrientation(Mesh Axis To Orient=<0 1 0>, Mesh Orientation Mode=::NewEnumerator2, Use Rotation Vector=false)
   /SetVariables(Particles.VisibilityTag=&MakeCustomIntFromBool(Bool=&SetBoolByFloatComparison(A=$System.Is Blue, B=0.5), True INT=1, False INT=0)) {isSetParameters=true}
  ]
  particleUpdateScript [
   &ParticleState(Kill Particles When Lifetime Has Elapsed=false, Let Infinitely Lived Particles Die When Emitter Deactivates=true)
   &ScaleMeshSize(Scale Factor=&VectorFromFloat(Value=$System.FX scale))
   &InterpolateOverTime(Rate Of Change=10, Target Vector Value:NiagaraPosition=$Particles.Position, Variable Type=::NewEnumerator2)
   /SetVariables(Particles.Interpolate Dynamic=&MakeFloat4FromVectorAndFloat("Vector (RGB)"=$Particles.InterpolateOverTime.InterpolateOverTime_V3_.Moving Average, "Float (Alpha)"=0.5)) {isSetParameters=true}
   &SolveForcesAndVelocity(Clamp Velocity=false, Limit Acceleration=false)
   &SkeletalMeshLocation(MaskSpawnGroup=false, OverrideRandomSeed=false, Skeletal Mesh=$System.Skeletal Mesh) {enabled=false}
  ]
  rendererClasses [&NiagaraMeshRendererProperties]
  renderers [
   {rendererClass=&NiagaraMeshRendererProperties, rendererIndex=0}
   {rendererClass=&NiagaraMeshRendererProperties, rendererIndex=1}
   {rendererClass=&NiagaraMeshRendererProperties, rendererIndex=2}
   {rendererClass=&NiagaraMeshRendererProperties, rendererIndex=3}
   {rendererClass=&NiagaraMeshRendererProperties, rendererIndex=4}
   {rendererClass=&NiagaraMeshRendererProperties, rendererIndex=5}
  ]
 }
 Minimal {
  emitterUpdateScript [
   &EmitterState(Inactive Response=::NewEnumerator1, Life Cycle Mode=::NewEnumerator1, Loop Duration=1, UseLoopDelay=false)
   &SpawnBurst_Instantaneous(Use Spawn Probability=false, UseLoopCountLimit=true)
   /SetVariables(
    Emitter.Linear Color:LinearColor=&LerpLinearColors(StartColor=<rgba 0.0041 0 0.0243 1>, EndColor=<rgba 0 0.0162 0.0243 1>, LerpFactor=$System.Is Blue)
    Emitter.Linear Color001:LinearColor=&LerpLinearColors(StartColor=<rgba 1 0.5237 0 1>, EndColor=<rgba 0 1 1.5 1>, LerpFactor=$System.Is Blue)
    Emitter.Linear Color002:LinearColor=&LerpLinearColors(StartColor=<rgba 0.1285 0.0033 0 1>, EndColor=<rgba 0 0.0856 0.1285 1>, LerpFactor=$System.Is Blue)
   ) {isSetParameters=true}
  ]
  particleSpawnScript [
   &InitializeParticle(
    Color Mode=::NewEnumerator0
    Lifetime=5
    Mesh Scale Mode=::NewEnumerator3
    Mesh Uniform Scale=&Multiply_Float(A=0.6, B=$System.FX scale)
    Position Offset=<0 0 100>
    Sprite Size Mode=::NewEnumerator3
    Uniform Sprite Size=1.745769414802391e+23
    UsePositionOffset=true
    Write Material Random=true
   )
   &InitialMeshOrientation(Mesh Orientation Mode=::NewEnumerator7, Rotation=<0 0 -30.25>, Use Rotation Vector=true)
  ]
  particleUpdateScript [
   &ParticleState(Kill Particles When Lifetime Has Elapsed=false, Let Infinitely Lived Particles Die When Emitter Deactivates=true)
   /SetVariables(Particles.VisibilityTag=&MakeCustomIntFromBool(
     Bool=&SetBoolByFloatComparison(
      A=&FloatFromCurve(
       FloatCurve=DataInterface() {propertyValues {Curve {keys [<at -1 -1>,<at 1 1>]}}}
       CurveIndex=&DotProduct(
        Vector A=&TransformVector(Vector=<1 0 0>, Destination Space=::Simulation)
        Vector B=&ReturnCameraVector(Camera Vectors=::NewEnumerator4, Coordinate Space=::Simulation, Origin Position=/Niagara/DynamicInputs/Helpers/SimulationPosition())
       )
      )
      B=0
     )
     True INT=1
     False INT=0
    )) {enabled=false, isSetParameters=true}
  ]
  rendererClasses [&NiagaraMeshRendererProperties]
  renderers [
   {rendererClass=&NiagaraMeshRendererProperties, rendererIndex=0}
   {rendererClass=&NiagaraMeshRendererProperties, rendererIndex=1}
  ]
 }
 TrailingEmbers {
  emitterUpdateScript [
   &EmitterState()
   &SpawnParticlesFromOtherEmitter(
    Attribute Reader=DataInterface() {propertyValues {EmitterBinding {bindingMode=Other, emitterName=TwistingSmoke}}}
    Maximum Age Spawn Threshold=0.2
    Minimum Age Spawn Threshold=0.5
    Spawn Rate=35
    Spawn Rate Per Particle Cap=300
    Use Maximum Age=false
    Use Minimum Age=false
   )
  ]
  particleSpawnScript [
   &InitializeParticle(
    Color=&LerpLinearColors(StartColor=<rgba 50 7.3503 0 1>, EndColor=<rgba 0 33.3333 50 1>, LerpFactor=$System.Is Blue)
    Lifetime=&RandomRangeFloat(Minimum=0.25, Maximum=0.5)
    Mass Mode=::NewEnumerator3
    Position Offset=<1 0 0>
    Sprite Rotation Mode=::NewEnumerator2
    Sprite Size Mode=::NewEnumerator1
    Uniform Sprite Size Max=1
    Uniform Sprite Size Min=0.25
    UsePositionOffset=true
    Write Material Random=true
   )
   &SampleParticlesFromOtherEmitter(
    Applied Position Offset:Vector3f=&VectorFromFloat(Value=&RandomRangeFloat(Minimum=-15, Maximum=15))
    Applied Velocity Scale:NiagaraFloat=&RandomRangeFloat(Minimum=2, Maximum=6)
    Emitter Sampling Mode:ENiagara_EmitterLocSamplingMode=::NewEnumerator0
    Maximum Sampled Particle Age=0.1
    Module.MaxSampledAgeCutoff=true
    Module.MinSampledAgeCutoff=false
    Reject Particle By Sampled Age=true
    Velocity Sampling=::NewEnumerator0
   )
   &InheritVelocity(
    Inherited Velocity Speed Limit:NiagaraFloat=&Multiply_Float(A=100, B=$System.FX scale)
    Speed Limit=true
    UseSpeedThreshold=false
   )
   &ShapeLocation(
    MaskSpawnGroup=false
    Non Uniform Scale=&Multiply_VectorByFloat(Vector=<3 1 1>, Float=$System.FX scale)
    Override Seed=false
    Sphere Radius=10
    UseOwnerScale=true
   )
  ]
  particleUpdateScript [
   &ParticleState()
   &Drag(Drag=2, Ignore Mass=false, Use Linear Drag=true, Use Rotational Drag=true)
   &GravityForce(Gravity=<0 0 5>)
   &CurlNoiseForce(Bias Noise=false, Noise Strength=100, Pan Noise=false)
   &SolveForcesAndVelocity(Clamp Velocity=false, Limit Acceleration=false)
   &ScaleSpriteSize(
    Uniform Curve Scale=2
    Uniform Curve Sprite Scale=DataInterface() {propertyValues {Curve {keys [
        {arriveTangent=9.1531, interpMode=RCIM_Cubic, leaveTangent=9.1531, tangentMode=RCTM_User, time=0, value=0.02}
        {arriveTangent=0.48, interpMode=RCIM_Cubic, leaveTangent=0.48, time=0.18, value=0.99}
        <at 1 0.5 RCIM_Cubic>
       ]}}}
   )
   &ScaleSpriteSize(
    Scale Sprite Size Mode=::NewEnumerator0
    Uniform Scale Factor=&FloatFromCurve(
     FloatCurve=DataInterface() {propertyValues {Curve {keys [{tangentMode=RCTM_User, time=0, value=0}, <at 211 211>, <at 220 75>, {tangentMode=RCTM_User, time=300, value=300}]}}}
     CurveIndex=$System.FX scale
    )
   ) {module=ScaleSpriteSize001}
   &ScaleSpriteSizeBySpeed(Max Scale Factor=<1 100>, Velocity Threshold:NiagaraFloat=&Multiply_Float(A=1000, B=$System.FX scale))
   &DynamicMaterialParameters(
    Index 0 Param 1=&FloatFromCurve(FloatCurve=DataInterface() {propertyValues {Curve {keys [<at 0 1>,<at 1 2>]}}})
    Index 0 Param 2=5
    Index 0 Param 3=1
    Index 0 Param 4=1
    Module.Param0WriteEnabled=true
    Module.Param1WriteEnabled=true
    Module.Param2WriteEnabled=true
    Module.Param3WriteEnabled=true
   )
  ]
  rendererClasses [&NiagaraSpriteRendererProperties]
  renderers [{rendererClass=&NiagaraSpriteRendererProperties}]
 }
 TrailingSmokeCore {
  emitterUpdateScript [
   &EmitterState()
   &SpawnBurst_Instantaneous(Spawn Count=100, Use Spawn Probability=false, UseLoopCountLimit=false) {enabled=false}
   &SpawnRate(SpawnRate=150, Use Spawn Probability=false)
  ]
  particleSpawnScript [
   &InitializeParticle(
    Color=&RandomRangeLinearColor(Minimum=<rgba 0 0 0 1>, Maximum=<rgba 0.0391 0.0436 0.0521 1>)
    Lifetime Max=1.2
    Lifetime Min=0.75
    Lifetime Mode=::NewEnumerator1
    Mass Mode=::NewEnumerator3
    Position Offset=&Multiply_VectorByFloat(Vector=<50 0 0>, Float=$System.FX scale)
    Position Offset Coordinate Space=::Local
    Sprite Rotation Mode=::NewEnumerator2
    Sprite Size Max=<7 10>
    Sprite Size Min=<5 12>
    Sprite Size Mode=::NewEnumerator2
    Sprite UV Mode=::NewEnumerator3
    UsePositionOffset=true
    Write Material Random=true
   )
   /SetVariables(Particles.RandomFrame=&RandomRangeInt(Minimum=0, Maximum=63)) {isSetParameters=true}
   &InheritVelocity(
    Inherited Velocity Amount Scale=&VectorFromFloat(Value=0.05)
    Speed Limit=false
    UseSpeedThreshold=false
   )
   &ShapeLocation(
    MaskSpawnGroup=false
    Override Seed=false
    Sphere Radius=&Multiply_Float(A=10, B=$System.FX scale)
    UseOwnerScale=true
   )
  ]
  particleUpdateScript [
   &ParticleState()
   &ScaleSpriteSize(Scale Sprite Size Mode=::NewEnumerator0, Uniform Scale Factor=$System.FX scale) {module=ScaleSpriteSize001}
   &GravityForce(Gravity=<0 0 5>)
   &Drag(Drag=0.75, Ignore Mass=false, Use Linear Drag=true, Use Rotational Drag=true)
   &CurlNoiseForce(
    Bias Noise=false
    Curl Noise Cone Mask Angle=60
    Curl Noise Cone Mask Falloff Angle=60
    Mask Curl Noise=true
    Noise Frequency=&Multiply_Float(A=50, B=$System.FX scale)
    Noise Strength=&Multiply_Float(A=100, B=$System.FX scale)
    Pan Noise=false
   )
   &ScaleColor(
    Scale Alpha=&FloatFromCurve(
     FloatCurve=DataInterface() {propertyValues {Curve {keys [
         <at 0 0 RCIM_Cubic>
         {interpMode=RCIM_Cubic, tangentMode=RCTM_User, time=0.25, value=1}
         {interpMode=RCIM_Cubic, tangentMode=RCTM_User, time=0.5, value=1}
         {tangentMode=RCTM_User, time=1, value=0}
        ]}}}
     Scale Curve=1.5
    )
    ScaleA=true
    ScaleRGB=true
   )
   &SolveForcesAndVelocity(Clamp Velocity=false, Limit Acceleration=false)
   &ScaleSpriteSize(Scale Sprite Size Mode=::NewEnumerator0, Uniform Scale Factor=7.5)
   &DynamicMaterialParameters(
    Index 0 Param 1=&FloatFromCurve(
     FloatCurve=DataInterface() {propertyValues {Curve {keys [
         {arriveTangent=2.8718, interpMode=RCIM_Cubic, leaveTangent=2.8718, tangentMode=RCTM_User, time=0, value=0}
         {arriveTangent=-0.3942, interpMode=RCIM_Cubic, leaveTangent=-0.3942, tangentMode=RCTM_User, time=1, value=1}
        ]}}}
     Scale Curve=0.1
    )
    Module.Index0Param0Write=true
    Module.Index0Param1Write=true
    Module.Index0Param2Write=true
    Module.Index0Param3Write=true
   )
   &FN_SubUVAnimation(End Frame=&Add_Integer(A=$Particles.RandomFrame, B=33), Start Frame=$Particles.RandomFrame)
  ]
  rendererClasses [&NiagaraSpriteRendererProperties]
  renderers [{rendererClass=&NiagaraSpriteRendererProperties}]
 }
 TwistingSmoke {
  emitterUpdateScript [
   &EmitterState()
   &SpawnRate(SpawnRate=20, Use Spawn Probability=false)
   &SpawnRate(SpawnGroup=1, SpawnRate=20, Use Spawn Probability=false) {module=SpawnRate001}
   /SetVariables(Emitter.C:LinearColor=&LerpLinearColors(StartColor=<rgba 3 0.5 0 1>, EndColor=<rgba 0 2 3 1>, LerpFactor=$System.Is Blue)) {enabled=false, isSetParameters=true, module=SetVariables_#10}
  ]
  particleSpawnScript [
   /SetVariables(Particles.NewAge:NiagaraFloat=&Add_Float(
     A=&Multiply_Float(A=$Emitter.Age, B=1)
     B=&Multiply_FloatByInt(Float=0.5, Integer=&SpawnGroup())
    )) {isSetParameters=true, module=SetVariables_#11}
   /SetVariables(Particles.Rad=0.2) {isSetParameters=true, module=SetVariables_#12}
   /SetVariables(Particles.OffsetDirection:Vector3f=&TransformVector(
     Vector=&MakeVector(
      X=10
      Y=&Sine(Normalized Angle=$Particles.NewAge, Period=1, Scale=$Particles.Rad, Bias=0)
      Z=&Cosine(Normalized Angle=$Particles.NewAge, Period=1, Scale=$Particles.Rad, Bias=0)
     )
     Destination Space=::Simulation
    )) {isSetParameters=true, module=SetVariables_#13}
   &InitializeParticle(
    Color=<rgba 0.0521 0.0665 0.0938 1>
    Position Offset=&Multiply_VectorByFloat(Vector=&Multiply_VectorByFloat(Vector=$Particles.OffsetDirection, Float=5), Float=$System.FX scale)
    Ribbon Width=150
    Ribbon Width Mode=::NewEnumerator1
    Sprite Size Mode=::NewEnumerator1
    UsePositionOffset=true
    Write Material Random=true
   )
   /SetVariables(Particles.RibbonID=/Niagara/DynamicInputs/Meshes/MakeRibbonIDFromSpawnGroup()) {enabled=false, isSetParameters=true, module=SetVariables_#14}
   /SetVariables(Particles.RibbonUVDistance:NiagaraFloat=&Divide_Float(A=$Emitter.DistanceTraveled, B=$System.FX scale)) {enabled=false, isSetParameters=true, module=SetVariables_#15}
   &InheritVelocity(
    Inherited Velocity Amount Scale=&VectorFromFloat(Value=0.01)
    Speed Limit=false
    UseSpeedThreshold=false
   )
   &AddVelocity(Velocity=<-100000 0 0>) {enabled=false}
  ]
  particleUpdateScript [
   &ParticleState()
   &ScaleSpriteSize(Scale Sprite Size Mode=::NewEnumerator0, Uniform Scale Factor=$System.FX scale) {enabled=false}
   &GravityForce(Gravity=&Multiply_VectorByFloat(Vector=$Particles.Initial.OffsetDirection, Float=30))
   &SolveForcesAndVelocity(Clamp Velocity=false, Limit Acceleration=false)
   &ScaleRibbonWidth(Ribbon Width Scale=&FloatFromCurve(
     FloatCurve=DataInterface() {propertyValues {Curve {keys [
         {arriveTangent=5.7729, interpMode=RCIM_Cubic, time=0, value=0}
         {arriveTangent=5, interpMode=RCIM_Cubic, leaveTangent=5, time=0.0711, value=0.4131}
         {interpMode=RCIM_Cubic, tangentMode=RCTM_User, time=0.2, value=1}
         {arriveTangent=0.1406, interpMode=RCIM_Cubic, leaveTangent=0.1406, tangentMode=RCTM_User, time=0.6, value=1}
         <at 1 0 RCIM_Cubic>
        ]}}}
     Scale Curve=0.35
    )) {enabled=false}
   &ScaleColor(
    Scale Alpha=&FloatFromCurve(
     FloatCurve=DataInterface() {propertyValues {Curve {keys [
         {interpMode=RCIM_Cubic, tangentMode=RCTM_User, time=0, value=0}
         {interpMode=RCIM_Cubic, tangentMode=RCTM_User, time=0.5, value=1}
         <at 1 0 RCIM_Cubic>
        ]}}}
     CurveIndex=&ReturnNormalizedExecIndex()
    )
    ScaleA=true
    ScaleRGB=true
   ) {enabled=false, module=ScaleColor001}
   /SetVariables(Particles.RibbonLinkOrder:NiagaraFloat=$Particles.NormalizedAge) {enabled=false, isSetParameters=true, module=SetVariables_#16}
   &DynamicMaterialParameters(
    Index 0 Param 1=&FloatFromCurve(FloatCurve=DataInterface() {propertyValues {Curve {keys [
         {interpMode=RCIM_Cubic, leaveTangent=-9.1594, tangentMode=RCTM_Break, time=0, value=1}
         {interpMode=RCIM_Cubic, leaveTangent=-5.9384, tangentMode=RCTM_Break, time=0.3, value=0}
        ]}}})
    Index 0 Param 2=&FloatFromCurve(FloatCurve=DataInterface() {propertyValues {Curve {keys [{tangentMode=RCTM_User, time=0, value=0}, {tangentMode=RCTM_User, time=1, value=1}]}}})
    Index 0 Param 4=&FloatFromCurve(FloatCurve=DataInterface() {propertyValues {Curve {keys [
         {interpMode=RCIM_Cubic, leaveTangent=-48.5828, tangentMode=RCTM_Break, time=0, value=3}
         {arriveTangent=-2.4798, tangentMode=RCTM_User, time=0.2, value=0.25}
         <at 0.5 0>
        ]}}})
    Module.Index0Param0Write=true
    Module.Index0Param1Write=true
    Module.Index0Param2Write=true
    Module.Index0Param3Write=true
   )
   &ScaleColor(
    Scale Alpha=&FloatFromCurve(
     FloatCurve=DataInterface() {propertyValues {Curve {keys [{tangentMode=RCTM_User, time=0, value=0}, {tangentMode=RCTM_User, time=0.5, value=1}]}}}
     CurveIndex=$Emitter.Age
    )
    ScaleA=true
    ScaleRGB=false
   ) {enabled=false}
   &ScaleRibbonWidth(Ribbon Width Scale=$System.FX scale) {enabled=false, module=ScaleRibbonWidth001}
  ]
  rendererClasses [&NiagaraRibbonRendererProperties]
  renderers [
   {rendererClass=&NiagaraRibbonRendererProperties}
   {rendererClass=&NiagaraRibbonRendererProperties, rendererIndex=1}
  ]
 }
 TwistingSmoke001 {
  emitterSpawnScript [/SetVariables(Emitter.RibDist:NiagaraFloat=&Multiply_Float(A=15, B=$System.FX scale)) {isSetParameters=true}]
  emitterUpdateScript [
   &EmitterState()
   &SpawnPerUnit(
    Module.Use Max Movement Threshold=false
    Module.Use Movement Tolerance=false
    Spawn Spacing:NiagaraFloat=$Emitter.RibDist
    Use Spawn Probability=false
   )
   &SpawnPerUnit(
    Module.Use Max Movement Threshold=false
    Module.Use Movement Tolerance=false
    Spawn Spacing:NiagaraFloat=$Emitter.RibDist
    Use Spawn Probability=false
   ) {module=SpawnPerUnit001}
   &SpawnRate(SpawnRate=20, Use Spawn Probability=false) {enabled=false}
   &SpawnRate(SpawnGroup=1, SpawnRate=20, Use Spawn Probability=false) {enabled=false, module=SpawnRate001}
   /SetVariables(Emitter.C:LinearColor=&LerpLinearColors(StartColor=<rgba 3 0.5 0 1>, EndColor=<rgba 0 2 3 1>, LerpFactor=$System.Is Blue)) {isSetParameters=true, module=SetVariables_#10}
  ]
  particleSpawnScript [
   /SetVariables(Particles.NewAge:NiagaraFloat=&Add_Float(
     A=&Multiply_Float(A=$Emitter.Age, B=1)
     B=&Multiply_FloatByInt(Float=0.5, Integer=&SpawnGroup())
    )) {isSetParameters=true, module=SetVariables_#11}
   /SetVariables(Particles.Rad=0.2) {isSetParameters=true, module=SetVariables_#12}
   /SetVariables(Particles.OffsetDirection:Vector3f=&TransformVector(
     Vector=&MakeVector(
      X=10
      Y=&Sine(Normalized Angle=$Particles.NewAge, Period=1, Scale=$Particles.Rad, Bias=0)
      Z=&Cosine(Normalized Angle=$Particles.NewAge, Period=1, Scale=$Particles.Rad, Bias=0)
     )
     Destination Space=::Simulation
    )) {isSetParameters=true, module=SetVariables_#13}
   &InitializeParticle(
    Color=<rgba 0.0521 0.0665 0.0938 1>
    Position Offset=&Multiply_VectorByFloat(Vector=&Multiply_VectorByFloat(Vector=$Particles.OffsetDirection, Float=5), Float=$System.FX scale)
    Ribbon Width=150
    Ribbon Width Mode=::NewEnumerator1
    Sprite Size Mode=::NewEnumerator1
    UsePositionOffset=true
    Write Material Random=true
   )
   /SetVariables(Particles.RibbonID=&MakeRibbonIDFromSpawnGroup()) {isSetParameters=true, module=SetVariables_#14}
   /SetVariables(Particles.RibbonUVDistance:NiagaraFloat=&Divide_Float(A=$Emitter.DistanceTraveled, B=$System.FX scale)) {isSetParameters=true, module=SetVariables_#15}
   &InheritVelocity(
    Inherited Velocity Amount Scale=&VectorFromFloat(Value=0.01)
    Speed Limit=false
    UseSpeedThreshold=false
   )
   &AddVelocity(Velocity=<-100000 0 0>) {enabled=false}
  ]
  particleUpdateScript [
   &ParticleState()
   &ScaleSpriteSize(Scale Sprite Size Mode=::NewEnumerator0, Uniform Scale Factor=$System.FX scale) {enabled=false}
   &GravityForce(Gravity=&Multiply_VectorByFloat(Vector=$Particles.Initial.OffsetDirection, Float=30))
   &SolveForcesAndVelocity(Clamp Velocity=false, Limit Acceleration=false)
   &ScaleRibbonWidth(Ribbon Width Scale=&FloatFromCurve(
     FloatCurve=DataInterface() {propertyValues {Curve {keys [
         {arriveTangent=5.7729, interpMode=RCIM_Cubic, time=0, value=0}
         {arriveTangent=5, interpMode=RCIM_Cubic, leaveTangent=5, time=0.0711, value=0.4131}
         {interpMode=RCIM_Cubic, tangentMode=RCTM_User, time=0.2, value=1}
         {arriveTangent=0.1406, interpMode=RCIM_Cubic, leaveTangent=0.1406, tangentMode=RCTM_User, time=0.6, value=1}
         <at 1 0 RCIM_Cubic>
        ]}}}
     Scale Curve=0.35
    ))
   &ScaleColor(
    Scale Alpha=&FloatFromCurve(
     FloatCurve=DataInterface() {propertyValues {Curve {keys [
         {
          arriveTangentWeight=0.4296
          interpMode=RCIM_Cubic
          leaveTangentWeight=0.4296
          tangentMode=RCTM_User
          tangentWeightMode=RCTWM_WeightedBoth
          time=0
          value=0
         }
         {
          arriveTangentWeight=0.1667
          interpMode=RCIM_Cubic
          leaveTangentWeight=0.1667
          tangentMode=RCTM_Break
          tangentWeightMode=RCTWM_WeightedBoth
          time=0.5
          value=1
         }
         {
          arriveTangentWeight=0.3969
          interpMode=RCIM_Cubic
          leaveTangentWeight=0.3969
          tangentMode=RCTM_User
          tangentWeightMode=RCTWM_WeightedBoth
          time=1
          value=0
         }
        ]}}}
     CurveIndex=&ReturnNormalizedExecIndex()
    )
    ScaleA=true
    ScaleRGB=true
   ) {module=ScaleColor001}
   /SetVariables(Particles.RibbonLinkOrder:NiagaraFloat=$Particles.NormalizedAge) {isSetParameters=true, module=SetVariables_#16}
   &DynamicMaterialParameters(
    Index 0 Param 1=&FloatFromCurve(FloatCurve=DataInterface() {propertyValues {Curve {keys [
         {interpMode=RCIM_Cubic, leaveTangent=-9.1594, tangentMode=RCTM_Break, time=0, value=1}
         {interpMode=RCIM_Cubic, leaveTangent=-5.9384, tangentMode=RCTM_Break, time=0.3, value=0}
        ]}}})
    Index 0 Param 2=&FloatFromCurve(FloatCurve=DataInterface() {propertyValues {Curve {keys [{tangentMode=RCTM_User, time=0, value=0}, {tangentMode=RCTM_User, time=1, value=1}]}}})
    Index 0 Param 4=&FloatFromCurve(FloatCurve=DataInterface() {propertyValues {Curve {keys [
         {interpMode=RCIM_Cubic, leaveTangent=-48.5828, tangentMode=RCTM_Break, time=0, value=3}
         {arriveTangent=-2.4798, tangentMode=RCTM_User, time=0.2, value=0.25}
         <at 0.5 0>
        ]}}})
    Module.Index0Param0Write=true
    Module.Index0Param1Write=true
    Module.Index0Param2Write=true
    Module.Index0Param3Write=true
   )
   &ScaleColor(
    Scale Alpha=&FloatFromCurve(
     FloatCurve=DataInterface() {propertyValues {Curve {keys [{tangentMode=RCTM_User, time=0, value=0}, {tangentMode=RCTM_User, time=0.5, value=1}]}}}
     CurveIndex=$Emitter.Age
    )
    ScaleA=true
    ScaleRGB=false
   )
   &ScaleRibbonWidth(Ribbon Width Scale=$System.FX scale) {module=ScaleRibbonWidth001}
  ]
  rendererClasses [&NiagaraRibbonRendererProperties]
  renderers [
   {rendererClass=&NiagaraRibbonRendererProperties}
   {rendererClass=&NiagaraRibbonRendererProperties, rendererIndex=1}
  ]
 }
}
summary {
 systemName=NS_GrillMoonFireBall
 emitters [
  {emitterName=Flare, rendererClasses [&NiagaraSpriteRendererProperties]}
  {emitterName=TwistingSmoke, rendererClasses [&NiagaraRibbonRendererProperties]}
  {emitterName=TrailingEmbers, rendererClasses [&NiagaraSpriteRendererProperties]}
  {emitterName=MaskFlame, rendererClasses [&NiagaraMeshRendererProperties]}
  {emitterName=TrailingSmokeCore, rendererClasses [&NiagaraSpriteRendererProperties]}
  {emitterName=Illumination, rendererClasses [&NiagaraSpriteRendererProperties, &NiagaraLightRendererProperties]}
  {emitterName=EyeTrail, rendererClasses [&NiagaraMeshRendererProperties]}
  {emitterName=Minimal, rendererClasses [&NiagaraMeshRendererProperties]}
  {emitterName=TwistingSmoke001, rendererClasses [&NiagaraRibbonRendererProperties]}
 ]
 userVariables [
  User.System_FX scale {type=NiagaraFloat, defaultValue=100, description=""}
  User.System_Is Blue {type=NiagaraFloat, defaultValue=0, description=""}
 ]
}
